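# Sound cue for the SleepyTime projectile pickup sound, part of the RiftBeacon RiftBubble event audio.
# A single wave player feeds the root directly with one pickup wave, boosted in volume and routed to the gadgets sound class.
# Spatialization references the shared SleepyTime attenuation asset; the cue's own attenuation values are stored without the override flag enabled.
# Its concurrency override caps simultaneous instances, stops the oldest and applies volume scaling to overlapping plays.
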SoundCue /Figment_S10/Athena/Sounds/Events/RiftBeacon/RiftBubble/SleepyTime/SleepyTime_Prj_Pickup_Cue
root=WavePlayer_1
nodes [{id=WavePlayer_1, properties {soundWaveAssetPtr=&/SleepyTime_Pickup_01}}]
properties {
 attenuationSettings=/FNGameplayCues/Athena/Sounds/Events/RiftBeacon/RiftBubble/SleepyTime/Fort_SleepyTime_Att
 bOverrideConcurrency=true
 maxDistance=107500
 soundClassObject=&Fort_Gadgets
 volumeMultiplier=1.5
 attenuationOverrides {distanceAlgorithm=NaturalSound, falloffDistance=3800, lPFRadiusMax=3000, lPFRadiusMin=2000, attenuationShapeExtents {x=2200}}
 concurrencyOverrides {maxCount=4, resolutionRule=StopOldest, volumeScale=0.5}
}
